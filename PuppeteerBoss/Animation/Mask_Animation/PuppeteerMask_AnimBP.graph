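# Animation Blueprint for the Puppeteer boss's mask, covering phase idles, fireball, teleport, eye hit, break and stagger reactions, and death.
# Cached poses chain state machines from Phase Idle through Player Left Arena and Action to additive Eye Hits, then DefaultSlot.
# Look blendspace and eye aim offset layer onto most non-stagger poses.
# Gameplay event and tag listeners latch phase 2 and pulse fireball and teleport flags; phase 2 transition callbacks toggle phase-change flags.
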AnimBlueprint /PuppeteerBoss/Animation/Mask_Animation/PuppeteerMask_AnimBP
target_skeleton=/PuppeteerBoss/Meshes/PuppeterMask_Creature_Skeleton
graphs [
 AnimGraph {nodes [
   AnimGraphNode_StateMachine_1 = Eye Hits SM() {is=title AnimGraphNode_StateMachine}
   Output Pose(Result=Slot 'DefaultSlot'(Source=@AnimGraphNode_StateMachine_1) {bAlwaysUpdateSourcePose=true})
   AnimGraphNode_StateMachine_0 = Phase Idle SM() {is=title AnimGraphNode_StateMachine, bAllowConduitEntryStates=true}
   Save cached pose 'Phase Idle Pose'(Pose=@AnimGraphNode_StateMachine_0) {cacheName=Phase Idle Pose}
   AnimGraphNode_StateMachine_2 = Action SM() {is=title AnimGraphNode_StateMachine}
   Save cached pose 'Action Pose'(Pose=@AnimGraphNode_StateMachine_2) {cacheName=Action Pose}
   AnimGraphNode_StateMachine_3 = Player Left Arena SM() {is=title AnimGraphNode_StateMachine}
   Save cached pose 'Player Left Arena Pose'(Pose=@AnimGraphNode_StateMachine_3) {cacheName=Player Left Arena Pose}
  ]}
 Phase Idle SM {
  entry=Pick Phase
  path=AnimGraph.AnimGraphNode_StateMachine_0.Phase Idle SM
  ids=id
  nodes [
   AnimStateConduitNode_1 = Pick Phase() {is=title AnimStateConduitNode}
   AnimStateNode_10 = Phase 0() {is=title AnimStateNode}
   AnimStateNode_11 = Phase Transition 1() {is=title AnimStateNode}
   AnimStateNode_7 = Phase 1() {is=title AnimStateNode}
   AnimStateNode_8 = Phase 2() {is=title AnimStateNode}
   AnimStateNode_9 = Phase Transition 2() {is=title AnimStateNode}
   AnimStateTransitionNode_15 = AnimStateTransitionNode() {crossfadeDuration=0.5}
   AnimStateTransitionNode_16 = AnimStateTransitionNode() {bAutomaticRuleBasedOnSequencePlayerInState=true, crossfadeDuration=0.6}
   AnimStateTransitionNode_17 = AnimStateTransitionNode() {priorityOrder=10}
   AnimStateTransitionNode_18 = AnimStateTransitionNode() {priorityOrder=10}
   AnimStateTransitionNode_19 = AnimStateTransitionNode()
   AnimStateTransitionNode_20 = AnimStateTransitionNode() {crossfadeDuration=0.5}
   AnimStateTransitionNode_21 = AnimStateTransitionNode() {bAutomaticRuleBasedOnSequencePlayerInState=true, crossfadeDuration=0.5}
  ]
  links [
   {
    dst=Phase Transition 1
    rule=AnimGraph.AnimGraphNode_StateMachine_0.Phase Idle SM.AnimStateTransitionNode_20.AnimationTransitionGraph_0
    src=Phase 0
    via=AnimStateTransitionNode_20
   }
   {
    dst=Phase Transition 2
    rule=AnimGraph.AnimGraphNode_StateMachine_0.Phase Idle SM.AnimStateTransitionNode_15.AnimationTransitionGraph_0
    src=Phase 1
    via=AnimStateTransitionNode_15
   }
   {
    dst=Phase 1
    rule=AnimGraph.AnimGraphNode_StateMachine_0.Phase Idle SM.AnimStateTransitionNode_21.AnimationTransitionGraph_0
    src=Phase Transition 1
    via=AnimStateTransitionNode_21
   }
   {
    dst=Phase 2
    rule=AnimGraph.AnimGraphNode_StateMachine_0.Phase Idle SM.AnimStateTransitionNode_16.AnimationTransitionGraph_0
    src=Phase Transition 2
    via=AnimStateTransitionNode_16
   }
   {
    dst=Phase 0
    rule=AnimGraph.AnimGraphNode_StateMachine_0.Phase Idle SM.AnimStateTransitionNode_19.AnimationTransitionGraph_0
    src=Pick Phase
    via=AnimStateTransitionNode_19
   }
   {
    dst=Phase 1
    rule=AnimGraph.AnimGraphNode_StateMachine_0.Phase Idle SM.AnimStateTransitionNode_17.AnimationTransitionGraph_0
    src=Pick Phase
    via=AnimStateTransitionNode_17
   }
   {
    dst=Phase 2
    rule=AnimGraph.AnimGraphNode_StateMachine_0.Phase Idle SM.AnimStateTransitionNode_18.AnimationTransitionGraph_0
    src=Pick Phase
    via=AnimStateTransitionNode_18
   }
  ]
 }
 Pick Phase {path=AnimGraph.AnimGraphNode_StateMachine_0.Phase Idle SM.AnimStateConduitNode_1.Pick Phase, nodes [Result(bCanEnterTransition=true)]}
 Phase 0 {
  path=AnimGraph.AnimGraphNode_StateMachine_0.Phase Idle SM.AnimStateNode_10.Phase 0
  nodes [Output Animation Pose(Result=Sequence Player 'PuppeteerBoss_Mask_Idle_Phase_00'() {sequence=&/PuppeteerBoss_Mask_Idle_Phase_00})]
 }
 Phase Transition 1 {
  path=AnimGraph.AnimGraphNode_StateMachine_0.Phase Idle SM.AnimStateNode_11.Phase Transition 1
  nodes [Output Animation Pose(Result=Sequence Player 'PuppeteerBoss_Mask_Phase_00_Into_Phase_01'() {bLoopAnimation=false, sequence=&/PuppeteerBoss_Mask_Phase_00_Into_Phase_01})]
 }
 Phase 1 {
  path=AnimGraph.AnimGraphNode_StateMachine_0.Phase Idle SM.AnimStateNode_7.Phase 1
  nodes [Output Animation Pose(Result=Sequence Player 'PuppeteerBoss_Mask_Idle_Phase_01'() {sequence=&/PuppeteerBoss_Mask_Idle_Phase_01})]
 }
 Phase 2 {
  path=AnimGraph.AnimGraphNode_StateMachine_0.Phase Idle SM.AnimStateNode_8.Phase 2
  nodes [Output Animation Pose(Result=Sequence Player 'PuppeteerBoss_Mask_Idle_Phase_02'() {sequence=&/PuppeteerBoss_Mask_Idle_Phase_02})]
 }
 Phase Transition 2 {
  path=AnimGraph.AnimGraphNode_StateMachine_0.Phase Idle SM.AnimStateNode_9.Phase Transition 2
  nodes [Output Animation Pose(Result=Sequence Player 'PuppeteerBoss_Mask_Phase_01_Into_Phase_02'() {bLoopAnimation=false, sequence=&/PuppeteerBoss_Mask_Phase_01_Into_Phase_02}) {
    node {stateEntryFunction {functionName=OnStateEntry}, stateExitFunction {functionName=OnStateExit}}
    stateEntryFunction {bSelfContext=true, memberGuid=#1, memberName=OnEnterPhaseChange}
    stateExitFunction {bSelfContext=true, memberGuid=#2, memberName=OnExitPhaseChange}
   }]
 }
 AnimationTransitionGraph_0 {
  path=AnimGraph.AnimGraphNode_StateMachine_0.Phase Idle SM.AnimStateTransitionNode_15.AnimationTransitionGraph_0
  nodes [Result(bCanEnterTransition=Get IsPhase2() -> (IsPhase2:bool))]
 }
 AnimationTransitionGraph_0 {path=AnimGraph.AnimGraphNode_StateMachine_0.Phase Idle SM.AnimStateTransitionNode_16.AnimationTransitionGraph_0, nodes [Result()]}
 AnimationTransitionGraph_0 {
  path=AnimGraph.AnimGraphNode_StateMachine_0.Phase Idle SM.AnimStateTransitionNode_17.AnimationTransitionGraph_0
  nodes [Result(bCanEnterTransition=NOT Boolean(A=Get IsPhase2()))]
 }
 AnimationTransitionGraph_0 {
  path=AnimGraph.AnimGraphNode_StateMachine_0.Phase Idle SM.AnimStateTransitionNode_18.AnimationTransitionGraph_0
  nodes [Result(bCanEnterTransition=Get IsPhase2())]
 }
 AnimationTransitionGraph_0 {
  path=AnimGraph.AnimGraphNode_StateMachine_0.Phase Idle SM.AnimStateTransitionNode_19.AnimationTransitionGraph_0
  nodes [Result(bCanEnterTransition=NOT Boolean(A=Get bIsStartPhase1() -> (bIsStartPhase1:bool)))]
 }
 AnimationTransitionGraph_0 {
  path=AnimGraph.AnimGraphNode_StateMachine_0.Phase Idle SM.AnimStateTransitionNode_20.AnimationTransitionGraph_0
  nodes [Result(bCanEnterTransition=Get bIsStartPhase1())]
 }
 AnimationTransitionGraph_0 {path=AnimGraph.AnimGraphNode_StateMachine_0.Phase Idle SM.AnimStateTransitionNode_21.AnimationTransitionGraph_0, nodes [Result()]}
 Eye Hits SM {
  entry=Passthrough
  path=AnimGraph.AnimGraphNode_StateMachine_1.Eye Hits SM
  ids=id
  nodes [
   To RightEyeHit() {is=title AnimStateAliasNode, also {aliases [LeftEyeHit,Passthrough]}}
   To LeftEyeHit() {is=title AnimStateAliasNode, also {aliases [Passthrough,RightEyeHit]}}
   AnimStateNode_0 = Passthrough() {is=title AnimStateNode}
   AnimStateNode_1 = RightEyeHit() {is=title AnimStateNode}
   AnimStateNode_2 = LeftEyeHit() {is=title AnimStateNode}
   AnimStateTransitionNode_0 = AnimStateTransitionNode() {crossfadeDuration=0}
   AnimStateTransitionNode_1 = AnimStateTransitionNode() {crossfadeDuration=0}
   AnimStateTransitionNode_13 = AnimStateTransitionNode() {bAutomaticRuleBasedOnSequencePlayerInState=true}
   AnimStateTransitionNode_14 = AnimStateTransitionNode() {bAutomaticRuleBasedOnSequencePlayerInState=true}
  ]
  links [
   {
    dst=Passthrough
    rule=AnimGraph.AnimGraphNode_StateMachine_1.Eye Hits SM.AnimStateTransitionNode_13.AnimationTransitionGraph_1
    src=LeftEyeHit
    via=AnimStateTransitionNode_13
   }
   {
    dst=Passthrough
    rule=AnimGraph.AnimGraphNode_StateMachine_1.Eye Hits SM.AnimStateTransitionNode_14.AnimationTransitionGraph_1
    src=RightEyeHit
    via=AnimStateTransitionNode_14
   }
   {
    dst=LeftEyeHit
    rule=AnimGraph.AnimGraphNode_StateMachine_1.Eye Hits SM.AnimStateTransitionNode_0.AnimationTransitionGraph_1
    src=To LeftEyeHit
    via=AnimStateTransitionNode_0
   }
   {
    dst=RightEyeHit
    rule=AnimGraph.AnimGraphNode_StateMachine_1.Eye Hits SM.AnimStateTransitionNode_1.AnimationTransitionGraph_0
    src=To RightEyeHit
    via=AnimStateTransitionNode_1
   }
  ]
 }
 Passthrough {
  path=AnimGraph.AnimGraphNode_StateMachine_1.Eye Hits SM.AnimStateNode_0.Passthrough
  nodes [Output Animation Pose(Result=Use cached pose 'Action Pose'())]
 }
 RightEyeHit {
  path=AnimGraph.AnimGraphNode_StateMachine_1.Eye Hits SM.AnimStateNode_1.RightEyeHit
  nodes [Output Animation Pose(Result=Apply Additive(
     Base=Use cached pose 'Action Pose'()
     Additive="Sequence Player (additive) 'PuppeteerBoss_Mask_HitReact_Right_Eye'"() {bLoopAnimation=false, sequence=&/PuppeteerBoss_Mask_HitReact_Right_Eye}
    ) {alphaBoolBlend {blendOption=Linear}})]
 }
 LeftEyeHit {
  path=AnimGraph.AnimGraphNode_StateMachine_1.Eye Hits SM.AnimStateNode_2.LeftEyeHit
  nodes [Output Animation Pose(Result=Apply Additive(
     Base=Use cached pose 'Action Pose'()
     Additive="Sequence Player (additive) 'PuppeteerBoss_Mask_HitReact_Left_Eye'"() {bLoopAnimation=false, sequence=&/PuppeteerBoss_Mask_HitReact_Left_Eye}
    ) {alphaBoolBlend {blendOption=Linear}})]
 }
 AnimationTransitionGraph_1 {
  path=AnimGraph.AnimGraphNode_StateMachine_1.Eye Hits SM.AnimStateTransitionNode_0.AnimationTransitionGraph_1
  nodes [Result(bCanEnterTransition=Get bIsLeftEyeHit() -> (bIsLeftEyeHit:bool))]
 }
 AnimationTransitionGraph_0 {
  path=AnimGraph.AnimGraphNode_StateMachine_1.Eye Hits SM.AnimStateTransitionNode_1.AnimationTransitionGraph_0
  nodes [Result(bCanEnterTransition=Get bIsRightEyeHit() -> (bIsRightEyeHit:bool))]
 }
 AnimationTransitionGraph_1 {path=AnimGraph.AnimGraphNode_StateMachine_1.Eye Hits SM.AnimStateTransitionNode_13.AnimationTransitionGraph_1, nodes [Result()]}
 AnimationTransitionGraph_1 {path=AnimGraph.AnimGraphNode_StateMachine_1.Eye Hits SM.AnimStateTransitionNode_14.AnimationTransitionGraph_1, nodes [Result()]}
 Action SM {
  entry=Idle
  path=AnimGraph.AnimGraphNode_StateMachine_2.Action SM
  ids=id
  nodes [
   To RightEyeStagger() {is=title AnimStateAliasNode, bGlobalAlias=true}
   To Death() {is=title AnimStateAliasNode, bGlobalAlias=true}
   To Teleport() {is=title AnimStateAliasNode, also {aliases [Fireball, Idle, LeftEyeBroken, PhaseChange, RightEyeBroken, Stagger Recovery, Teleport]}}
   To Fireball() {is=title AnimStateAliasNode, bGlobalAlias=true}
   To LeftEyeBroken() {is=title AnimStateAliasNode, bGlobalAlias=true}
   To RightEyeBroken() {is=title AnimStateAliasNode, bGlobalAlias=true}
   To LeftEyeStagger() {is=title AnimStateAliasNode, bGlobalAlias=true}
   To PhaseChange() {is=title AnimStateAliasNode, bGlobalAlias=true}
   AnimStateNode_0 = Idle()
   AnimStateNode_1 = Fireball() {is=title AnimStateNode}
   AnimStateNode_10 = PhaseChange() {is=title AnimStateNode}
   AnimStateNode_2 = Death() {is=title AnimStateNode}
   AnimStateNode_3 = Teleport() {is=title AnimStateNode}
   AnimStateNode_4 = Stagger Idle() {is=title AnimStateNode}
   AnimStateNode_5 = LeftEyeBroken() {is=title AnimStateNode}
   AnimStateNode_6 = RightEyeBroken() {is=title AnimStateNode}
   AnimStateNode_7 = LeftEyeStagger() {is=title AnimStateNode}
   AnimStateNode_8 = RightEyeStagger() {is=title AnimStateNode}
   AnimStateNode_9 = Stagger Recovery() {is=title AnimStateNode}
   AnimStateTransitionNode_0 = AnimStateTransitionNode() {crossfadeDuration=0.4}
   AnimStateTransitionNode_1 = AnimStateTransitionNode() {crossfadeDuration=0.3}
   AnimStateTransitionNode_10 = AnimStateTransitionNode() {bAutomaticRuleBasedOnSequencePlayerInState=true, crossfadeDuration=0.5}
   AnimStateTransitionNode_11 = AnimStateTransitionNode() {bAutomaticRuleBasedOnSequencePlayerInState=true, crossfadeDuration=0.5}
   AnimStateTransitionNode_12 = AnimStateTransitionNode() {crossfadeDuration=0.15}
   AnimStateTransitionNode_13 = AnimStateTransitionNode() {bAutomaticRuleBasedOnSequencePlayerInState=true, crossfadeDuration=0.4}
   AnimStateTransitionNode_14 = AnimStateTransitionNode() {bAutomaticRuleBasedOnSequencePlayerInState=true, crossfadeDuration=0.4}
   AnimStateTransitionNode_15 = AnimStateTransitionNode() {crossfadeDuration=0.5}
   AnimStateTransitionNode_16 = AnimStateTransitionNode() {crossfadeDuration=0.8}
   AnimStateTransitionNode_2 = AnimStateTransitionNode() {crossfadeDuration=0.01}
   AnimStateTransitionNode_3 = AnimStateTransitionNode() {crossfadeDuration=0.3}
   AnimStateTransitionNode_4 = AnimStateTransitionNode() {crossfadeDuration=0.01}
   AnimStateTransitionNode_5 = AnimStateTransitionNode() {bAutomaticRuleBasedOnSequencePlayerInState=true, crossfadeDuration=0.5}
   AnimStateTransitionNode_6 = AnimStateTransitionNode() {bAutomaticRuleBasedOnSequencePlayerInState=true, crossfadeDuration=0.4}
   AnimStateTransitionNode_7 = AnimStateTransitionNode() {bAutomaticRuleBasedOnSequencePlayerInState=true, crossfadeDuration=0.3}
   AnimStateTransitionNode_8 = AnimStateTransitionNode() {blendMode=Linear, crossfadeDuration=0.01}
   AnimStateTransitionNode_9 = AnimStateTransitionNode() {crossfadeDuration=0.15}
  ]
  links [
   {
    dst=Idle
    rule=AnimGraph.AnimGraphNode_StateMachine_2.Action SM.AnimStateTransitionNode_5.AnimationTransitionGraph_0
    src=Fireball
    via=AnimStateTransitionNode_5
   }
   {
    dst=Idle
    rule=AnimGraph.AnimGraphNode_StateMachine_2.Action SM.AnimStateTransitionNode_13.Transition
    src=LeftEyeBroken
    via=AnimStateTransitionNode_13
   }
   {
    dst=Stagger Idle
    rule=AnimGraph.AnimGraphNode_StateMachine_2.Action SM.AnimStateTransitionNode_10.AnimationTransitionGraph_1
    src=LeftEyeStagger
    via=AnimStateTransitionNode_10
   }
   {
    dst=Idle
    rule=AnimGraph.AnimGraphNode_StateMachine_2.Action SM.AnimStateTransitionNode_16.Transition
    src=PhaseChange
    via=AnimStateTransitionNode_16
   }
   {
    dst=Idle
    rule=AnimGraph.AnimGraphNode_StateMachine_2.Action SM.AnimStateTransitionNode_14.Transition
    src=RightEyeBroken
    via=AnimStateTransitionNode_14
   }
   {
    dst=Stagger Idle
    rule=AnimGraph.AnimGraphNode_StateMachine_2.Action SM.AnimStateTransitionNode_11.AnimationTransitionGraph_1
    src=RightEyeStagger
    via=AnimStateTransitionNode_11
   }
   {
    dst=Stagger Recovery
    rule=AnimGraph.AnimGraphNode_StateMachine_2.Action SM.AnimStateTransitionNode_0.Transition
    src=Stagger Idle
    via=AnimStateTransitionNode_0
   }
   {
    dst=Idle
    rule=AnimGraph.AnimGraphNode_StateMachine_2.Action SM.AnimStateTransitionNode_6.Transition
    src=Stagger Recovery
    via=AnimStateTransitionNode_6
   }
   {
    dst=Idle
    rule=AnimGraph.AnimGraphNode_StateMachine_2.Action SM.AnimStateTransitionNode_7.AnimationTransitionGraph_0
    src=Teleport
    via=AnimStateTransitionNode_7
   }
   {
    dst=Death
    rule=AnimGraph.AnimGraphNode_StateMachine_2.Action SM.AnimStateTransitionNode_3.AnimationTransitionGraph_0
    src=To Death
    via=AnimStateTransitionNode_3
   }
   {
    dst=Fireball
    rule=AnimGraph.AnimGraphNode_StateMachine_2.Action SM.AnimStateTransitionNode_1.AnimationTransitionGraph_0
    src=To Fireball
    via=AnimStateTransitionNode_1
   }
   {
    dst=LeftEyeBroken
    rule=AnimGraph.AnimGraphNode_StateMachine_2.Action SM.AnimStateTransitionNode_2.AnimationTransitionGraph_1
    src=To LeftEyeBroken
    via=AnimStateTransitionNode_2
   }
   {
    dst=LeftEyeStagger
    rule=AnimGraph.AnimGraphNode_StateMachine_2.Action SM.AnimStateTransitionNode_9.AnimationTransitionGraph_1
    src=To LeftEyeStagger
    via=AnimStateTransitionNode_9
   }
   {
    dst=PhaseChange
    rule=AnimGraph.AnimGraphNode_StateMachine_2.Action SM.AnimStateTransitionNode_15.Transition
    src=To PhaseChange
    via=AnimStateTransitionNode_15
   }
   {
    dst=RightEyeBroken
    rule=AnimGraph.AnimGraphNode_StateMachine_2.Action SM.AnimStateTransitionNode_4.AnimationTransitionGraph_1
    src=To RightEyeBroken
    via=AnimStateTransitionNode_4
   }
   {
    dst=RightEyeStagger
    rule=AnimGraph.AnimGraphNode_StateMachine_2.Action SM.AnimStateTransitionNode_12.AnimationTransitionGraph_1
    src=To RightEyeStagger
    via=AnimStateTransitionNode_12
   }
   {
    dst=Teleport
    rule=AnimGraph.AnimGraphNode_StateMachine_2.Action SM.AnimStateTransitionNode_8.AnimationTransitionGraph_0
    src=To Teleport
    via=AnimStateTransitionNode_8
   }
  ]
 }
 Idle {
  path=AnimGraph.AnimGraphNode_StateMachine_2.Action SM.AnimStateNode_0.Idle
  nodes [
   n1 = Apply Additive(
    Base=Use cached pose 'Phase Idle Pose'()
    Additive=Blendspace Player 'PuppeteerBoss_Look_BS'(X=0, Y=0) {blendSpace=&/PuppeteerBoss_Look_BS}
    Alpha=Get LookBSAlpha() -> (LookBSAlpha:float)
   ) {alphaBoolBlend {blendOption=Linear}}
   n2 = Use cached pose 'Phase Idle Pose'()
   Modify Curve(
    SourcePose=AimOffset Player 'PuppeteerBoss_Mask_Eye_AO'!(
     X:float=0
     Y:float=0
     BasePose:Pose Link Structure=Layered blend per bone(BasePose=@n2, BlendPoses_0=@n1) {layerSetup [{branchFilters [{boneName=Mask}]}]}
     Alpha:float=Get LookBSAlpha()
    ) -> (Pose:Pose Link Structure) {blendSpace=&/PuppeteerBoss_Mask_Eye_AO, alphaBoolBlend {blendOption=Linear}}
    CurveValues_0=1
    Alpha=1
   ) {curveNames [UserCurve02], curveValues [0]}
   Output Animation Pose(Result=Use cached pose 'Player Left Arena Pose'())
  ]
 }
 Fireball {
  path=AnimGraph.AnimGraphNode_StateMachine_2.Action SM.AnimStateNode_1.Fireball
  nodes [
   n1 = Sequence Player 'PuppeteerBoss_Mask_Projectile_Attack'() {
    bLoopAnimation=false
    groupName=Fireball
    method=SyncGroup
    playRate=0.8
    sequence=&/PuppeteerBoss_Mask_Projectile_Attack
    startPosition=1.5
   }
   n2 = Sequence Player 'PuppeteerBoss_Mask_Projectile_Attack'() {
    bLoopAnimation=false
    groupName=Fireball
    method=SyncGroup
    playRate=0.8
    sequence=&PuppeteerBoss_Mask_Projectile_Attack
    startPosition=1.5
   }
   n3 = Layered blend per bone(
    BasePose=@n2
    BlendPoses_0=Apply Additive(
     Base=@n1
     Additive=Blendspace Player 'PuppeteerBoss_Look_BS'(X=0, Y=0) {blendSpace=&PuppeteerBoss_Look_BS}
     Alpha=Get LookBSAlpha()
    ) {alphaBoolBlend {blendOption=Linear}}
   ) {layerSetup [{branchFilters [{boneName=Mask}]}]}
   n4 = Get LookBSAlpha()
   Output Animation Pose(Result=Modify Curve(
     SourcePose=AimOffset Player 'PuppeteerBoss_Mask_Eye_AO'(X=0, Y=0, BasePose=@n3, Alpha=@n4) {blendSpace=&PuppeteerBoss_Mask_Eye_AO, alphaBoolBlend {blendOption=Linear}}
     CurveValues_0=1
     Alpha=1
    ) {curveNames [UserCurve02], curveValues [0]})
  ]
 }
 PhaseChange {
  path=AnimGraph.AnimGraphNode_StateMachine_2.Action SM.AnimStateNode_10.PhaseChange
  nodes [
   n1 = Layered blend per bone(
    BasePose=Use cached pose 'Phase Idle Pose'()
    BlendPoses_0=Apply Additive(
     Base=Use cached pose 'Phase Idle Pose'()
     Additive=Blendspace Player 'PuppeteerBoss_Look_BS'(X=0, Y=0) {blendSpace=&PuppeteerBoss_Look_BS}
     Alpha=Get LookBSAlpha()
    ) {alphaBoolBlend {blendOption=Linear}}
   ) {layerSetup [{branchFilters [{boneName=Mask}]}]}
   n2 = Get LookBSAlpha()
   Output Animation Pose(Result=Modify Curve(
     SourcePose=AimOffset Player 'PuppeteerBoss_Mask_Eye_AO'(X=0, Y=0, BasePose=@n1, Alpha=@n2) {blendSpace=&PuppeteerBoss_Mask_Eye_AO, alphaBoolBlend {blendOption=Linear}}
     CurveValues_0=0.5
     Alpha=1
    ) {curveNames [UserCurve02], curveValues [0]})
  ]
 }
 Death {
  path=AnimGraph.AnimGraphNode_StateMachine_2.Action SM.AnimStateNode_2.Death
  nodes [Output Animation Pose(Result=Sequence Player 'PuppeteerBoss_Mask_Death'() {bLoopAnimation=false, sequence=&/PuppeteerBoss_Mask_Death})]
 }
 Teleport {
  path=AnimGraph.AnimGraphNode_StateMachine_2.Action SM.AnimStateNode_3.Teleport
  nodes [
   n1 = Blendspace Player 'PuppeteerBoss_Look_BS'(X=0, Y=0) {blendSpace=&PuppeteerBoss_Look_BS}
   n2 = Sequence Player 'PuppeteerBoss_Mask_Appear_Phase_01'() {bLoopAnimation=false, sequence=&/PuppeteerBoss_Mask_Appear_Phase_01}
   n3 = Get LookBSAlpha()
   Output Animation Pose(Result=AimOffset Player 'PuppeteerBoss_Mask_Eye_AO'(X=0, Y=0, BasePose=Apply Additive(Base=@n2, Additive=@n1, Alpha=@n3) {alphaBoolBlend {blendOption=Linear}}, Alpha=Get LookBSAlpha()) {blendSpace=&PuppeteerBoss_Mask_Eye_AO, alphaBoolBlend {blendOption=Linear}})
  ]
 }
 Stagger Idle {
  path=AnimGraph.AnimGraphNode_StateMachine_2.Action SM.AnimStateNode_4.Stagger Idle
  nodes [Output Animation Pose(Result=Sequence Player 'PuppeteerBoss_Mask_Stagger_Idle'() {bLoopAnimation=false, sequence=&/PuppeteerBoss_Mask_Stagger_Idle})]
 }
 LeftEyeBroken {
  path=AnimGraph.AnimGraphNode_StateMachine_2.Action SM.AnimStateNode_5.LeftEyeBroken
  nodes [
   n1 = Blendspace Player 'PuppeteerBoss_Look_BS'(X=0, Y=0) {blendSpace=&PuppeteerBoss_Look_BS}
   n2 = Sequence Player 'PuppeteerBoss_Mask_Break_Left_Eye'() {bLoopAnimation=false, sequence=&/PuppeteerBoss_Mask_Break_Left_Eye}
   n3 = Get LookBSAlpha()
   Output Animation Pose(Result=AimOffset Player 'PuppeteerBoss_Mask_Eye_AO'(X=0, Y=0, BasePose=Apply Additive(Base=@n2, Additive=@n1, Alpha=@n3) {alphaBoolBlend {blendOption=Linear}}, Alpha=Get LookBSAlpha()) {blendSpace=&PuppeteerBoss_Mask_Eye_AO, alphaBoolBlend {blendOption=Linear}})
  ]
 }
 RightEyeBroken {
  path=AnimGraph.AnimGraphNode_StateMachine_2.Action SM.AnimStateNode_6.RightEyeBroken
  nodes [
   n1 = Blendspace Player 'PuppeteerBoss_Look_BS'(X=0, Y=0) {blendSpace=&PuppeteerBoss_Look_BS}
   n2 = Sequence Player 'PuppeteerBoss_Mask_Break_Right_Eye'() {bLoopAnimation=false, sequence=&/PuppeteerBoss_Mask_Break_Right_Eye}
   n3 = Get LookBSAlpha()
   Output Animation Pose(Result=AimOffset Player 'PuppeteerBoss_Mask_Eye_AO'(X=0, Y=0, BasePose=Apply Additive(Base=@n2, Additive=@n1, Alpha=@n3) {alphaBoolBlend {blendOption=Linear}}, Alpha=Get LookBSAlpha()) {blendSpace=&PuppeteerBoss_Mask_Eye_AO, alphaBoolBlend {blendOption=Linear}})
  ]
 }
 LeftEyeStagger {
  path=AnimGraph.AnimGraphNode_StateMachine_2.Action SM.AnimStateNode_7.LeftEyeStagger
  nodes [Output Animation Pose(Result=Sequence Player 'PuppeteerBoss_Mask_Stagger_Left_Eye'() {bLoopAnimation=false, sequence=&/PuppeteerBoss_Mask_Stagger_Left_Eye})]
 }
 RightEyeStagger {
  path=AnimGraph.AnimGraphNode_StateMachine_2.Action SM.AnimStateNode_8.RightEyeStagger
  nodes [Output Animation Pose(Result=Sequence Player 'PuppeteerBoss_Mask_Stagger_Right_Eye'() {bLoopAnimation=false, sequence=&/PuppeteerBoss_Mask_Stagger_Right_Eye})]
 }
 Stagger Recovery {
  path=AnimGraph.AnimGraphNode_StateMachine_2.Action SM.AnimStateNode_9.Stagger Recovery
  nodes [
   n1 = Blendspace Player 'PuppeteerBoss_Look_BS'(X=0, Y=0) {blendSpace=&PuppeteerBoss_Look_BS}
   n2 = Sequence Player 'PuppeteerBoss_Mask_Stagger_Recovery'() {bLoopAnimation=false, sequence=&/PuppeteerBoss_Mask_Stagger_Recovery}
   n3 = Get LookBSAlpha()
   Output Animation Pose(Result=AimOffset Player 'PuppeteerBoss_Mask_Eye_AO'(X=0, Y=0, BasePose=Apply Additive(Base=@n2, Additive=@n1, Alpha=@n3) {alphaBoolBlend {blendOption=Linear}}, Alpha=Get LookBSAlpha()) {blendSpace=&PuppeteerBoss_Mask_Eye_AO, alphaBoolBlend {blendOption=Linear}})
  ]
 }
 Transition {
  path=AnimGraph.AnimGraphNode_StateMachine_2.Action SM.AnimStateTransitionNode_0.Transition
  nodes [Result(bCanEnterTransition=NOT Boolean(A=Get bIsStaggered() -> (bIsStaggered:bool)))]
 }
 AnimationTransitionGraph_0 {
  path=AnimGraph.AnimGraphNode_StateMachine_2.Action SM.AnimStateTransitionNode_1.AnimationTransitionGraph_0
  nodes [Result(bCanEnterTransition=Get IsFireball() -> (IsFireball:bool))]
 }
 AnimationTransitionGraph_1 {path=AnimGraph.AnimGraphNode_StateMachine_2.Action SM.AnimStateTransitionNode_10.AnimationTransitionGraph_1, nodes [Result()]}
 AnimationTransitionGraph_1 {path=AnimGraph.AnimGraphNode_StateMachine_2.Action SM.AnimStateTransitionNode_11.AnimationTransitionGraph_1, nodes [Result()]}
 AnimationTransitionGraph_1 {
  path=AnimGraph.AnimGraphNode_StateMachine_2.Action SM.AnimStateTransitionNode_12.AnimationTransitionGraph_1
  nodes [Result(bCanEnterTransition=Get bIsRightEyeStagger() -> (bIsRightEyeStagger:bool))]
 }
 Transition {path=AnimGraph.AnimGraphNode_StateMachine_2.Action SM.AnimStateTransitionNode_13.Transition, nodes [Result()]}
 Transition {path=AnimGraph.AnimGraphNode_StateMachine_2.Action SM.AnimStateTransitionNode_14.Transition, nodes [Result()]}
 Transition {
  path=AnimGraph.AnimGraphNode_StateMachine_2.Action SM.AnimStateTransitionNode_15.Transition
  nodes [Result(bCanEnterTransition=Get IsPhaseChanging() -> (IsPhaseChanging:bool))]
 }
 Transition {
  path=AnimGraph.AnimGraphNode_StateMachine_2.Action SM.AnimStateTransitionNode_16.Transition
  nodes [Result(bCanEnterTransition=NOT Boolean(A=Get IsPhaseChanging()))]
 }
 AnimationTransitionGraph_1 {
  path=AnimGraph.AnimGraphNode_StateMachine_2.Action SM.AnimStateTransitionNode_2.AnimationTransitionGraph_1
  nodes [Result(bCanEnterTransition=Get bIsLeftEyeBroken() -> (bIsLeftEyeBroken:bool))]
 }
 AnimationTransitionGraph_0 {
  path=AnimGraph.AnimGraphNode_StateMachine_2.Action SM.AnimStateTransitionNode_3.AnimationTransitionGraph_0
  nodes [Result(bCanEnterTransition=Get bIsDeath() -> (bIsDeath:bool))]
 }
 AnimationTransitionGraph_1 {
  path=AnimGraph.AnimGraphNode_StateMachine_2.Action SM.AnimStateTransitionNode_4.AnimationTransitionGraph_1
  nodes [Result(bCanEnterTransition=Get bIsRightEyeBroken() -> (bIsRightEyeBroken:bool))]
 }
 AnimationTransitionGraph_0 {path=AnimGraph.AnimGraphNode_StateMachine_2.Action SM.AnimStateTransitionNode_5.AnimationTransitionGraph_0, nodes [Result()]}
 Transition {path=AnimGraph.AnimGraphNode_StateMachine_2.Action SM.AnimStateTransitionNode_6.Transition, nodes [Result()]}
 AnimationTransitionGraph_0 {path=AnimGraph.AnimGraphNode_StateMachine_2.Action SM.AnimStateTransitionNode_7.AnimationTransitionGraph_0, nodes [Result()]}
 AnimationTransitionGraph_0 {
  path=AnimGraph.AnimGraphNode_StateMachine_2.Action SM.AnimStateTransitionNode_8.AnimationTransitionGraph_0
  nodes [Result(bCanEnterTransition=Get IsTeleport() -> (IsTeleport:bool))]
 }
 AnimationTransitionGraph_1 {
  path=AnimGraph.AnimGraphNode_StateMachine_2.Action SM.AnimStateTransitionNode_9.AnimationTransitionGraph_1
  nodes [Result(bCanEnterTransition=Get bIsLeftEyeStagger() -> (bIsLeftEyeStagger:bool))]
 }
 Player Left Arena SM {
  entry=Pass Through
  path=AnimGraph.AnimGraphNode_StateMachine_3.Player Left Arena SM
  ids=id
  nodes [
   AnimStateNode_0 = Pass Through() {is=title AnimStateNode}
   AnimStateNode_1 = Player Left Arena() {is=title AnimStateNode}
   AnimStateTransitionNode_8 = AnimStateTransitionNode() {crossfadeDuration=1}
   AnimStateTransitionNode_9 = AnimStateTransitionNode() {crossfadeDuration=1}
  ]
  links [
   {
    dst=Player Left Arena
    rule=AnimGraph.AnimGraphNode_StateMachine_3.Player Left Arena SM.AnimStateTransitionNode_8.Transition
    src=Pass Through
    via=AnimStateTransitionNode_8
   }
   {
    dst=Pass Through
    rule=AnimGraph.AnimGraphNode_StateMachine_3.Player Left Arena SM.AnimStateTransitionNode_9.Transition
    src=Player Left Arena
    via=AnimStateTransitionNode_9
   }
  ]
 }
 Pass Through {
  path=AnimGraph.AnimGraphNode_StateMachine_3.Player Left Arena SM.AnimStateNode_0.Pass Through
  nodes [
   n1 = Layered blend per bone(
    BasePose=Use cached pose 'Phase Idle Pose'()
    BlendPoses_0=Apply Additive(
     Base=Use cached pose 'Phase Idle Pose'()
     Additive=Blendspace Player 'PuppeteerBoss_Look_BS'(X=0, Y=0) {blendSpace=&PuppeteerBoss_Look_BS}
     Alpha=Get LookBSAlpha()
    ) {alphaBoolBlend {blendOption=Linear}}
   ) {layerSetup [{branchFilters [{boneName=Mask}]}]}
   n2 = Get LookBSAlpha()
   Output Animation Pose(Result=Modify Curve(
     SourcePose=AimOffset Player 'PuppeteerBoss_Mask_Eye_AO'(X=0, Y=0, BasePose=@n1, Alpha=@n2) {blendSpace=&PuppeteerBoss_Mask_Eye_AO, alphaBoolBlend {blendOption=Linear}}
     CurveValues_0=1
     Alpha=1
    ) {curveNames [UserCurve02], curveValues [0]})
  ]
 }
 Player Left Arena {
  path=AnimGraph.AnimGraphNode_StateMachine_3.Player Left Arena SM.AnimStateNode_1.Player Left Arena
  nodes [Output Animation Pose(Result=Sequence Player 'PuppeteerBoss_Mask_Idle_Phase_00'() {sequence=&PuppeteerBoss_Mask_Idle_Phase_00})]
 }
 Transition {
  path=AnimGraph.AnimGraphNode_StateMachine_3.Player Left Arena SM.AnimStateTransitionNode_8.Transition
  nodes [Result(bCanEnterTransition=Get bIsPlayerOutOfArena() -> (bIsPlayerOutOfArena:bool))]
 }
 Transition {
  path=AnimGraph.AnimGraphNode_StateMachine_3.Player Left Arena SM.AnimStateTransitionNode_9.Transition
  nodes [Result(bCanEnterTransition=NOT Boolean(A=Get bIsPlayerOutOfArena()))]
 }
 EventGraph {nodes [
   n1 = OnAbilitySystemComponentInitialized_Event() {is=title K2Node_CustomEvent}
   n2 = Get As BGA Puppeteer Mask() -> (As BGA Puppeteer Mask:BGA Puppeteer Mask Object Reference)
   > n3 = WaitGameplayEventToActor!(
    execute:Exec
    TargetActor:Actor Object Reference=@n2
    EventTag:Gameplay Tag Structure=(TagName="Ability.PuppeteerBoss.Mask.ShootFireballs")
    OnlyTriggerOnce:bool=false
    OnlyMatchExact:bool=true
   ) -> (then:Exec, AsyncTaskProxy:Ability Async Wait Gameplay Event Object Reference, EventReceived:Exec, Payload:Gameplay Event Data Structure) {is=title K2Node_AsyncAction}
   > n4 = WaitGameplayTagAddToActor!(
    execute:Exec
    TargetActor:Actor Object Reference=@n2
    Tag:Gameplay Tag Structure=(TagName="PuppeteerBoss.Demon.IsInPhase2")
    OnlyTriggerOnce:bool=false
   ) -> (then:Exec, AsyncTaskProxy:Ability Async Wait Gameplay Tag Added Object Reference, Added:Exec) {is=title K2Node_AsyncAction}
   > n5 = WaitGameplayTagAddToActor(TargetActor=@n2, Tag=(TagName="PuppeteerBoss.Mask.Ability.Teleport.Activate"), OnlyTriggerOnce=false)
   Delay(execute=@n5.Added, Duration=0.2)
   > Set IsTeleport(IsTeleport:bool=true) -> (Output_Get:bool)
   > DelayUntilNextTick!(execute) -> (then)
   > Set IsTeleport(IsTeleport=false)
   Set IsPhase2(execute=@n4.Added, IsPhase2:bool=true) -> (Output_Get:bool)
   Set IsFireball(execute=@n3.EventReceived, IsFireball:bool=true) -> (Output_Get:bool)
   > DelayUntilNextTick()
   > Set IsFireball(IsFireball=false)
   Event BlueprintInitializeAnimation()
   > n6 = Cast To BGA_PuppeteerMask!(execute, Object=GetOwningActor()) -> (then, CastFailed, AsBGA Puppeteer Mask:BGA Puppeteer Mask Object Reference) {is=title K2Node_DynamicCast}
   > n7 = Set As BGA Puppeteer Mask(As BGA Puppeteer Mask:BGA Puppeteer Mask Object Reference=@n6) -> (Output_Get:BGA Puppeteer Mask Object Reference)
   > Assign On Ability System Component Initialized!(execute:Exec, self:Building Actor Object Reference=@n7, Delegate:"Delegate (by ref)"=@n1) -> (then:Exec) {is=title K2Node_AssignDelegate}
   > Set bIsLookAtDemon(bIsLookAtDemon:bool=true) -> (Output_Get:bool)
  ]}
 OnEnterPhaseChange {nodes [
   OnEnterPhaseChange!() -> (then, Context:"Anim Update Context Structure (by ref)", Node:"Anim Node Reference Structure (by ref)") {is=title K2Node_FunctionEntry}
   > Set bIsLookAtDemon(bIsLookAtDemon=true)
   > Set IsPhaseChanging(IsPhaseChanging:bool=true) -> (Output_Get:bool)
   > Return Node()
  ]}
 OnExitPhaseChange {nodes [
   OnExitPhaseChange!() -> (then, Context:"Anim Update Context Structure (by ref)", Node:"Anim Node Reference Structure (by ref)") {is=title K2Node_FunctionEntry}
   > Set bIsLookAtDemon(bIsLookAtDemon=false)
   > Set IsPhaseChanging(IsPhaseChanging=false)
   > Return Node()
  ]}
]
variables [
 As BGA Puppeteer Mask {type=BGA_PuppeteerMask_C, default=null}
 IsFireball {type=boolean}
 IsPhase2 {type=boolean}
 IsPhaseChanging {type=boolean}
 IsTeleport {type=boolean}
]
